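# Stereo one-shot MetaSound source for the UI sound of discovering a point of interest on the map.
# A single Wave Player starts on play, feeds its left and right channels to the outputs, and signals finished when the wave ends.
# The wave is an exposed input defaulting to a POI discovered sound.
# It uses the non-critical in-game UI sound class and a POI discovery concurrency rule.
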MetaSoundSource /Game/Sounds/Fort_UI_Sounds/Gameplay/Discovery/MSS_UI_POI_Map_Discovery
inputs [
 Sound {type=WaveAsset}
 UE.Source.OnPlay {type=Trigger}
]
interfaces [UE.OutputFormat.Stereo:1.0, UE.Source.OneShot:1.0, UE.Source:1.1]
outputs [
 "UE.OutputFormat.Stereo.Audio:0" {type=Audio}
 "UE.OutputFormat.Stereo.Audio:1" {type=Audio}
 UE.Source.OneShot.OnFinished {type=Trigger}
]
pages [Default {
  input_defaults {Sound=&/UI_Hud_POI_Discovered_03, UE.Source.OnPlay=null}
  nodes [WavePlayer_0 = UE.Wave Player:Stereo(Wave Asset=^Sound, Play=^UE.Source.OnPlay) {also {version [1,1]}}]
  links [
   {dst="@UE.Source.OneShot.OnFinished", src=WavePlayer_0.On Finished}
   {dst="@UE.OutputFormat.Stereo.Audio:0", src=WavePlayer_0.Out Left}
   {dst="@UE.OutputFormat.Stereo.Audio:1", src=WavePlayer_0.Out Right}
  ]
 }]
properties {
 assetImportData=&self:AssetImportData
 numChannels=2
 outputFormat=Stereo
 soundClassObject=&Fort_UI_InGame_NonCritical
 concurrencySet [/Game/Sounds/Concurrency_Rules/CONC_Discovery_POI]
}
